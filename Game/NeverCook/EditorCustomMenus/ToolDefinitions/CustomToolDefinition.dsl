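; CustomToolDefinition is an editor-only tool definition for custom editor menus, exposing a friendly name, tooltip and tool command targets through getter functions.
; Its level and widget getters check the soft references, convert them to path strings and split at the last dot to yield package names.
; The "validated" name and tooltip getters return their variables unchanged, with no actual validation.
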(fn GetLevelAssetPackageName
  (bind _levelreference (Variables|ToolCommand|GetLevelReference))
  (bind _worldasset (Utilities|Struct|BreakSoftWorldReference _levelreference))
  (if (Utilities|IsValidSoftObjectReference _worldasset)
    (return (Utilities|String|Split (Utilities|BreakSoftObjectPath (Utilities|ToSoftObjectPath _worldasset)) "." "IgnoreCase" "FromEnd"))))

(fn GetValidatedFriendlyName
  (return (Variables|Default|GetFriendlyName)))

(fn GetValidatedTooltip
  (return (Variables|Default|GetToolTip)))

(fn GetWidgetAssetPackageName
  (bind _widgetassetreference (Variables|ToolCommand|GetWidgetAssetReference))
  (bind _returnvalue_pathstring (Utilities|BreakSoftClassPath (Utilities|ToSoftClassPath _widgetassetreference)))
  (if (Utilities|IsValidSoftClassReference _widgetassetreference)
    (return (Utilities|String|Split _returnvalue_pathstring "." "IgnoreCase" "FromEnd"))))
